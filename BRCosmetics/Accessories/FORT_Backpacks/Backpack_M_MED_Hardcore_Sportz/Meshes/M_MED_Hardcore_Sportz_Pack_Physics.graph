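# Physics asset for the Hardcore Sportz backpack cosmetic, giving box bodies to the pack and its dangling ribbon bones.
# Two five-segment chains, top and bottom, are constrained link to link from a shared dyn_ribbon_parent body.
# Twist is tightly limited on every joint, while swing varies, with the bottom chain tighter near its root.
# The large pack box and small dyn_ribbon_base body have no constraints.
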PhysicsAsset /BRCosmetics/Accessories/FORT_Backpacks/Backpack_M_MED_Hardcore_Sportz/Meshes/M_MED_Hardcore_Sportz_Pack_Physics
bodies {
 dyn_ribbon_base [{shapeType=Box, extentX=0.5, extentY=0.5, extentZ=0.5}]
 dyn_ribbon_bottom_01 [{shapeType=Box, extentX=1.76, extentY=2.01, extentZ=0.75}]
 dyn_ribbon_bottom_02 [{shapeType=Box, extentX=3.0287, extentY=1.8511, extentZ=1, center=<1.4827 0 0>}]
 dyn_ribbon_bottom_03 [{shapeType=Box, extentX=3.3646, extentY=1.1585, extentZ=2.8919, center=<1.6068 0 0>}]
 dyn_ribbon_bottom_04 [{shapeType=Box, extentX=4.4176, extentY=2.9339, extentZ=1.25, center=<1.9923 0 0>}]
 dyn_ribbon_bottom_05 [{shapeType=Box, extentX=7.1674, extentY=2.6954, extentZ=2.5, center=<3.053 0.0001 -1.6407>, rotation=<rot -31.3848 0 0>}]
 dyn_ribbon_parent [{shapeType=Box, extentX=0.5, extentY=0.5, extentZ=0.5}]
 dyn_ribbon_top_01 [{shapeType=Box, extentX=3.01, extentY=1.76, extentZ=2, center=<1.4299 0 -0.2765>, rotation=<rot -10.1977 0 0>}]
 dyn_ribbon_top_02 [{shapeType=Box, extentX=5.2446, extentY=2.8124, extentZ=1.5, center=<2.6546 -0.2777 -0.091>, rotation=<rot -4.3445 -6.4606 43.2936>}]
 dyn_ribbon_top_03 [{shapeType=Box, extentX=4.7461, extentY=2.8177, extentZ=1.3249, center=<2.4971 0 0>}]
 dyn_ribbon_top_04 [{shapeType=Box, extentX=4.8372, extentY=3.6147, extentZ=1.593, center=<2.523 0.155 0.2344>, rotation=<rot 5.3571 3.5528 33.6253>}]
 dyn_ribbopn_top_05 [{shapeType=Box, extentX=8.202, extentY=4.106, extentZ=1.5982, center=<3.741 1.3008 -0.5823>, rotation=<rot -8.6449 24.081 0>}]
 pack [{shapeType=Box, extentX=31.25, extentY=7.5, extentZ=20.25, center=<-13.141 6.6675 -0.0877>, rotation=<rot 0 0 2.1763>}]
}
constraints [
 {bone1Name=dyn_ribbon_top_01, bone2Name=dyn_ribbon_parent, swing1LimitDegrees=30, swing2LimitDegrees=15, twistLimitDegrees=3}
 {bone1Name=dyn_ribbon_top_02, bone2Name=dyn_ribbon_top_01, swing1LimitDegrees=25, swing2LimitDegrees=35, twistLimitDegrees=3}
 {bone1Name=dyn_ribbon_top_03, bone2Name=dyn_ribbon_top_02, swing1LimitDegrees=35, swing2LimitDegrees=30, twistLimitDegrees=3}
 {bone1Name=dyn_ribbon_top_04, bone2Name=dyn_ribbon_top_03, swing1LimitDegrees=35, swing2LimitDegrees=30, twistLimitDegrees=3}
 {bone1Name=dyn_ribbopn_top_05, bone2Name=dyn_ribbon_top_04, swing1LimitDegrees=30, swing2LimitDegrees=35, twistLimitDegrees=3}
 {bone1Name=dyn_ribbon_bottom_01, bone2Name=dyn_ribbon_parent, swing1LimitDegrees=10, swing2LimitDegrees=20, twistLimitDegrees=3}
 {bone1Name=dyn_ribbon_bottom_02, bone2Name=dyn_ribbon_bottom_01, swing1LimitDegrees=10, swing2LimitDegrees=10, twistLimitDegrees=3}
 {bone1Name=dyn_ribbon_bottom_03, bone2Name=dyn_ribbon_bottom_02, swing1LimitDegrees=10, swing2LimitDegrees=25, twistLimitDegrees=3}
 {bone1Name=dyn_ribbon_bottom_04, bone2Name=dyn_ribbon_bottom_03, swing1LimitDegrees=30, swing2LimitDegrees=20, twistLimitDegrees=3}
 {bone1Name=dyn_ribbon_bottom_05, bone2Name=dyn_ribbon_bottom_04, swing1LimitDegrees=30, swing2LimitDegrees=20, twistLimitDegrees=3}
]
